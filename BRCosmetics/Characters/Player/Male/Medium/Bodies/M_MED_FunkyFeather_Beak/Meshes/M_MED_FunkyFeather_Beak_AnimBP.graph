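# Animation Blueprint for the FunkyFeather Beak player body cosmetic, which follows the attached parent mesh and animates a floating duck on it.
# The copied pose gets a duck offset and a Control Rig pass, then partial head and neck rotations computed by negating and scaling the head's delta from reference pose.
# Time-driven sine waves bob and rock base_duck, and spring-based AnimDynamics on duck_dyn adds damped secondary motion.
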AnimBlueprint /BRCosmetics/Characters/Player/Male/Medium/Bodies/M_MED_FunkyFeather_Beak/Meshes/M_MED_FunkyFeather_Beak_AnimBP
target_skeleton=&/M_MED_FunkyFeather_Beak_Skeleton
graphs [
 AnimGraph {nodes [
   n1 = "Transform (Modify) Bone - Bone: base_duck"!(
    ComponentPose=Local To Component(LocalPose=Copy Pose From Mesh() {bUseAttachedParent=true})
    bAlphaBoolEnabled:bool=True
    Translation="-4,0,0"
    Rotation="0,0,0"
    Scale="1,1,1"
   ) -> (Pose) {
    alphaInputType=Bool
    rotationMode=BMM_Replace
    rotationSpace=BCS_BoneSpace
    translationMode=BMM_Additive
    translationSpace=BCS_BoneSpace
    alphaBoolBlend {blendInTime=0.2, blendOption=Linear, blendOutTime=0.2}
    boneToModify {boneName=base_duck}
    translation {x=-4}
   }
   n2 = Control Rig(Source=Component To Local(ComponentPose=@n1)) {controlRigAssetReference {controlRigAsset=&/M_MED_FunkyFeather_Beak_CtrlRig}}
   n3 = "Transform (Modify) Bone - Bone: head"!(ComponentPose=Local To Component(LocalPose=@n2), Alpha=0.5, Rotation=Get Head Rot() -> (Head Rot:Rotator)) -> (Pose) {
    rotationMode=BMM_Replace
    rotationSpace=BCS_BoneSpace
    translationSpace=BCS_BoneSpace
    alphaBoolBlend {blendInTime=0.1, blendOption=Linear, blendOutTime=0.1}
    boneToModify {boneName=head}
   }
   n4 = Get Neck Rot() -> (Neck Rot:Rotator)
   n5 = "Transform (Modify) Bone - Bone: neck_01"!(ComponentPose=@n3, Alpha=0.8, Rotation=@n4) -> (Pose) {
    rotationMode=BMM_Replace
    rotationSpace=BCS_BoneSpace
    translationSpace=BCS_BoneSpace
    alphaBoolBlend {blendInTime=0.1, blendOption=Linear, blendOutTime=0.1}
    boneToModify {boneName=neck_01}
   }
   n6 = "Transform (Modify) Bone - Bone: neck_02"!(ComponentPose=@n5, Alpha=0.8, Rotation=@n4) -> (Pose) {
    rotationMode=BMM_Replace
    rotationSpace=BCS_BoneSpace
    translationSpace=BCS_BoneSpace
    alphaBoolBlend {blendInTime=0.1, blendOption=Linear, blendOutTime=0.1}
    boneToModify {boneName=neck_02}
   }
   n7 = "Transform (Modify) Bone - Bone: base_duck"(
    ComponentPose=@n6
    bAlphaBoolEnabled=True
    Translation=Get Duck_1 Floating() -> (Duck_1 Floating:Vector)
    Rotation=Get Duck_1 Rotator() -> (Duck_1 Rotator:Rotator)
    Scale="1,1,1"
   ) {
    alphaInputType=Bool
    rotationMode=BMM_Replace
    rotationSpace=BCS_BoneSpace
    translationMode=BMM_Additive
    translationSpace=BCS_BoneSpace
    alphaBoolBlend {blendInTime=0.2, blendOption=Linear, blendOutTime=0.2}
    boneToModify {boneName=base_duck}
   }
   n8 = Anim Dynamics - Bone: duck_dyn(ComponentPose=@n7, Alpha=1, ExternalForce="0,0,0") {
    angularBiasOverride=0.99
    angularDampingOverride=0.99
    angularSpringConstant=10
    bLinearSpring=true
    bOverrideLinearDamping=true
    bUseGravityOverride=true
    gravityScale=0.15
    linearDampingOverride=0.9999
    linearSpringConstant=25
    simulationSpace=RootRelative
    alphaBoolBlend {blendOption=Linear}
    alphaScaleBiasClamp {bInterpResult=true, interpSpeedDecreasing=5, interpSpeedIncreasing=5}
    boundBone {boneName=duck_dyn}
    componentAppliedLinearAccClamp=<40 40 40>
    componentLinearAccScale=<0.2 0.2 0.1>
    componentLinearVelScale=<0.2 0.2 0.1>
    physicsBodyDefinitions [{
      boundBone {boneName=duck_dyn}
      boxExtents=<16 16 16>
      constraintSetup {angularLimitsMax=<0 0 0>, angularLimitsMin=<0 0 0>, angularTarget=<1 0 0>, linearAxesMax=<3 0.2 0.2>, linearAxesMin=<-1 -0.2 -0.2>}
      localJointOffset=<0 0 0>
     }]
   }
   Output Pose(Result=Component To Local(ComponentPose=@n8)) {layerGroup=None}
  ]}
 EventGraph {nodes [
   n1 = Event BlueprintUpdateAnimation()
   > Set Delta Time X(Delta Time X=@n1.DeltaTimeX)
   n2 = BreakRotator(InRot=GetDeltaTransformFromRefPose!(self=GetOwnerSkeletalMeshComponent(), BoneName=head, BaseName=neck_02) -> (ReturnValue_Location:Vector, ReturnValue_Rotation:Rotator, ReturnValue_Scale:Vector).ReturnValue_Rotation)
   n3 = Remap Float(Input=@n2.Yaw, OldRange=(X=10,Y=-28), NewRange=(X=0,Y=1), Clamp=true) -> (Output=~)
   n4 = Lerp(A=Get Z Head Rot Per Min() -> (Z Head Rot Per Min:double), B=Get Z Head Rot Per Max() -> (Z Head Rot Per Max:double), Alpha=@n3)
   n5 = float * float!(A=@n2, B=-1, C:double=Get X Head Rot Per() -> (X Head Rot Per:double)) -> (ReturnValue) {is=title K2Node_PromotableOperator}
   n6 = float * float(A=@n2.Pitch, B=-1, C=Get Y Head Rot Per() -> (Y Head Rot Per:double))
   > Set Head Rot(Head Rot:Rotator=MakeRotator(Roll=@n5, Pitch=@n6, Yaw=float * float(A=@n2.Yaw, B=-1, C=@n4))) -> (Output_Get:Rotator)
   n7 = Lerp(A=Get Z Head Rot Per Min(), B=0.5, Alpha=@n3)
   > Set Neck Rot(Neck Rot:Rotator=MakeRotator(Roll=@n5, Pitch=@n6, Yaw=float * float(B=-1, C=@n7))) -> (Output_Get:Rotator)
   > Sequence()
   n8 = float * float!(A=GetGameTimeInSeconds(), B=3) -> (ReturnValue)
   n9 = "Sin (Radians)"(A=float * float(A=0.5, B=@n8) {is=title K2Node_CommutativeAssociativeBinaryOperator})
   n10 = "Sin (Radians)"(A=float * float(A=0.4, B=@n8))
   > Set Duck_1 Floating(Duck_1 Floating:Vector=MakeVector(X=float * float(A=@n10, B=3), Y=0, Z=float * float(A=@n9, B=0.3))) -> (Output_Get:Vector)
   n11 = float * float(A=GetGameTimeInSeconds(), B=4) {is=title K2Node_PromotableOperator}
   n12 = "Sin (Radians)"(A=float * float(A=0.32, B=@n11) {is=title K2Node_CommutativeAssociativeBinaryOperator})
   n13 = "Sin (Radians)"(A=float * float(A=0.16, B=@n11))
   > Set Duck_1 Rotator(Duck_1 Rotator:Rotator=MakeRotator(Roll=float * float(A=@n12, B=22), Yaw=float * float(A=@n13, B=0.32))) -> (Output_Get:Rotator)
  ]}
]
variables [
 AlphaJump {type=number}
 Delta Time X {type=number}
 Duck_1 Floating {type=Vector}
 Duck_1 Rotator {type=Rotator}
 Duck_2 Floating {type=Vector}
 Head Rot {type=Rotator}
 Neck Rot {type=Rotator}
 X Head Rot Per {type=number, default=0.4}
 Y Head Rot Per {type=number, default=0.3}
 Z Head Rot Per Max {type=number, default=0.5}
 Z Head Rot Per Min {type=number}
]
