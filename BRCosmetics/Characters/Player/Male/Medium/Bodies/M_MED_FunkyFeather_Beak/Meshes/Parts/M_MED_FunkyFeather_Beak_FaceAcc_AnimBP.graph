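# Post-copy AnimBP for the FunkyFeather Beak face accessory, animating two floating, spinning duck parts.
# Copies the parent mesh pose, offsets both duck bases, then runs Control Rig, RigidBody relative to spine_05, and spring-based AnimDynamics.
# Update event drives opposite time-based spins and sine bob and tilt, with floating vectors cross-wired between ducks.
# It also computes Gravity Override, which the graph never reads, as RigidBody gravity is fixed.
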AnimBlueprint /BRCosmetics/Characters/Player/Male/Medium/Bodies/M_MED_FunkyFeather_Beak/Meshes/Parts/M_MED_FunkyFeather_Beak_FaceAcc_AnimBP
target_skeleton=&/M_MED_FunkyFeather_Beak_FaceAcc_Skeleton
graphs [
 AnimGraph {nodes [
   n1 = "Transform (Modify) Bone - Bone: Base_duck_01"(
    ComponentPose=Local To Component(LocalPose=Copy Pose From Mesh() {bUseAttachedParent=true})
    Alpha=1
    Translation="0,0,0"
    Rotation="0,-99,0"
    Scale="1,1,1"
   ) {rotationMode=BMM_Additive, rotationSpace=BCS_BoneSpace, boneToModify {boneName=Base_duck_01}, rotation {yaw=-99}}
   n2 = Control Rig(Source=Component To Local(ComponentPose="Transform (Modify) Bone - Bone: Base_duck_02"(ComponentPose=@n1, Alpha=1, Translation="0,0,0", Rotation="0,99,0", Scale="1,1,1") {rotationMode=BMM_Additive, rotationSpace=BCS_BoneSpace, boneToModify {boneName=Base_duck_02}, rotation {yaw=99}})) {controlRigAssetReference {controlRigAsset=&/M_MED_FunkyFeather_Beak_FaceAcc_CtrlRig}}
   n3 = RigidBody!(ComponentPose=Local To Component(LocalPose=@n2), Alpha, OverrideWorldGravity:Vector="0,0,-980", ExternalForce) -> (Pose) {
    bClampLinearTranslationLimitToRefPose=true
    bOverrideWorldGravity=true
    overridePhysicsAsset=&/M_MED_FunkyFeather_Beak_FaceAcc_Physics
    simulationSpace=BaseBoneSpace
    baseBoneRef {boneName=spine_05}
    overrideWorldGravity {z=-980}
    simSpaceSettings {dampingAlpha=0, maxAngularAcceleration=80, maxAngularVelocity=12, maxLinearAcceleration=2000, maxLinearVelocity=500, worldAlpha=1}
   }
   n4 = "Transform (Modify) Bone - Bone: rotate_duck_01"!(
    ComponentPose=@n3
    Alpha=1
    Translation="0,0,0"
    Rotation_Roll:float=Get CurrentRotation() -> (CurrentRotation:double)
    Rotation_Pitch:float=0
    Rotation_Yaw:float=0
    Scale="1,1,1"
   ) -> (Pose) {
    rotationMode=BMM_Additive
    rotationSpace=BCS_BoneSpace
    translationMode=BMM_Additive
    alphaBoolBlend {blendOption=Linear}
    boneToModify {boneName=rotate_duck_01}
   }
   n5 = "Transform (Modify) Bone - Bone: rotate_duck_02"!(
    ComponentPose=@n4
    Alpha=1
    Translation="0,0,0"
    Rotation_Roll:float=Get CurrentRotation_Duck() -> (CurrentRotation_Duck:double)
    Rotation_Pitch:float=0
    Rotation_Yaw:float=0
    Scale="1,1,1"
   ) -> (Pose) {rotationMode=BMM_Additive, rotationSpace=BCS_BoneSpace, alphaBoolBlend {blendOption=Linear}, boneToModify {boneName=rotate_duck_02}}
   n6 = Anim Dynamics - Bone: Base_duck_01(ComponentPose=@n5, Alpha=1, ExternalForce="0,0,0") {
    angularBiasOverride=0.99
    angularDampingOverride=0.99
    angularSpringConstant=10
    bLinearSpring=true
    bOverrideLinearDamping=true
    bUseGravityOverride=true
    gravityScale=0.15
    linearDampingOverride=0.9999
    linearSpringConstant=40
    simulationSpace=RootRelative
    alphaBoolBlend {blendOption=Linear}
    alphaScaleBiasClamp {bInterpResult=true, interpSpeedDecreasing=5, interpSpeedIncreasing=5}
    boundBone {boneName=Base_duck_01}
    componentAppliedLinearAccClamp=<40 40 40>
    componentLinearAccScale=<0.2 0.2 0.1>
    componentLinearVelScale=<0.2 0.2 0.1>
    physicsBodyDefinitions [{
      boundBone {boneName=Base_duck_01}
      boxExtents=<44 44 44>
      constraintSetup {angularLimitsMax=<0 0 0>, angularLimitsMin=<0 0 0>, angularTarget=<1 0 0>, linearAxesMax=<0.8 0.3 0.3>, linearAxesMin=<-0.8 -0.3 -0.3>}
      localJointOffset=<0 0 0>
     }]
   }
   n7 = Anim Dynamics - Bone: duck_02_dyn(ComponentPose=@n6, Alpha=1, ExternalForce="0,0,0") {
    angularBiasOverride=0.99
    angularDampingOverride=0.99
    angularSpringConstant=10
    bLinearSpring=true
    bOverrideLinearDamping=true
    bUseGravityOverride=true
    gravityScale=0.15
    linearDampingOverride=0.9999
    linearSpringConstant=40
    simulationSpace=RootRelative
    alphaBoolBlend {blendOption=Linear}
    alphaScaleBiasClamp {bInterpResult=true, interpSpeedDecreasing=5, interpSpeedIncreasing=5}
    boundBone {boneName=duck_02_dyn}
    componentAppliedLinearAccClamp=<40 40 40>
    componentLinearAccScale=<0.2 0.2 0.1>
    componentLinearVelScale=<0.2 0.2 0.1>
    physicsBodyDefinitions [{
      boundBone {boneName=duck_02_dyn}
      boxExtents=<44 44 44>
      constraintSetup {angularLimitsMax=<0 0 0>, angularLimitsMin=<0 0 0>, angularTarget=<1 0 0>, linearAxesMax=<0.8 0.3 0.3>, linearAxesMin=<-0.8 -0.3 -0.3>}
      localJointOffset=<0 0 0>
     }]
   }
   n8 = "Transform (Modify) Bone - Bone: Base_duck_01"(
    ComponentPose=@n7
    Alpha=1
    Translation=Get Duck_2 Floating() -> (Duck_2 Floating:Vector)
    Rotation=Get Duck_1 rotator() -> (Duck_1 rotator:Rotator)
    Scale="1,1,1"
   ) {rotationMode=BMM_Additive, translationMode=BMM_Additive, boneToModify {boneName=Base_duck_01}}
   n9 = Get Duck_2 rotator() -> (Duck_2 rotator:Rotator)
   n10 = Get Duck_1 Floating() -> (Duck_1 Floating:Vector)
   Output Pose(Result=Component To Local(ComponentPose="Transform (Modify) Bone - Bone: Base_duck_02"(ComponentPose=@n8, Alpha=1, Translation=@n10, Rotation=@n9, Scale="1,1,1") {rotationMode=BMM_Additive, translationMode=BMM_Additive, boneToModify {boneName=Base_duck_02}})) {layerGroup=None}
  ]}
 EventGraph {nodes [
   Event BlueprintUpdateAnimation()
   > n1 = "Gravity Override State Function (Parameter Asset) (Nativized)"(
    CharacterPartAnimInstance=Self-Reference()
    GameTimeInSeconds=GetGameTimeInSeconds()
    Parameters=/BRCosmetics/Characters/Player/Male/Medium/Bodies/M_MED_FunkyFeather_Beak/Meshes/M_MED_FunkyFeather_Beak_Parameters
   )
   > Set Gravity Override(Gravity Override=@n1)
   n2 = GetRealTimeSeconds()
   > Set CurrentRotation(CurrentRotation:double="Division (Whole and Remainder)"!(Dividend:double=float * float(A=@n2, B=20) {is=title K2Node_PromotableOperator}, Divisor:double=360) -> (Remainder:double, ReturnValue:int=0)) -> (Output_Get:double)
   > Set CurrentRotation_Duck(CurrentRotation_Duck:double="Division (Whole and Remainder)"(Dividend=float * float(A=@n2, B=-13), Divisor=360) -> (ReturnValue=0)) -> (Output_Get:double)
   n3 = float * float(A=GetGameTimeInSeconds(), B=3)
   n4 = "Sin (Radians)"(A=float * float(A=0.3, B=@n3) {is=title K2Node_CommutativeAssociativeBinaryOperator})
   n5 = "Sin (Radians)"(A=float * float(A=0.3, B=@n3))
   n6 = "Sin (Radians)"(A=float * float(A=0.3, B=@n3))
   > Set Duck_1 Floating(Duck_1 Floating:Vector=MakeVector(X=float * float(A=@n6, B=0.7), Y=float * float(A=@n5, B=0.75), Z=float * float(A=@n4, B=1))) -> (Output_Get:Vector)
   n7 = float * float(A=GetGameTimeInSeconds(), B=3) {is=title K2Node_PromotableOperator}
   n8 = "Sin (Radians)"(A=float * float(A=0.5, B=@n7) {is=title K2Node_CommutativeAssociativeBinaryOperator})
   n9 = "Sin (Radians)"(A=float * float(A=0.5, B=@n7))
   n10 = "Sin (Radians)"(A=float * float(A=0.5, B=@n7))
   > Set Duck_2 Floating(Duck_2 Floating:Vector=MakeVector(X=float * float(A=@n10, B=0.7), Y=float * float(A=@n9, B=0.55), Z=float * float(A=@n8, B=1))) -> (Output_Get:Vector)
   n11 = float * float(A=GetGameTimeInSeconds(), B=3) {is=title K2Node_PromotableOperator}
   n12 = "Sin (Radians)"(A=float * float(A=0.3, B=@n11) {is=title K2Node_CommutativeAssociativeBinaryOperator})
   n13 = "Sin (Radians)"(A=float * float(A=0.35, B=@n11))
   > Set Duck_1 rotator(Duck_1 rotator:Rotator=MakeRotator(Roll=float * float(A=@n13, B=22), Pitch=float * float(A=@n12, B=22))) -> (Output_Get:Rotator)
   n14 = float * float(A=GetGameTimeInSeconds(), B=3) {is=title K2Node_PromotableOperator}
   n15 = "Sin (Radians)"(A=float * float(A=0.25, B=@n14) {is=title K2Node_CommutativeAssociativeBinaryOperator})
   n16 = "Sin (Radians)"(A=float * float(A=0.2, B=@n14))
   > Set Duck_2 rotator(Duck_2 rotator:Rotator=MakeRotator(Roll=float * float(A=@n16, B=22), Pitch=float * float(A=@n15, B=22))) -> (Output_Get:Rotator)
  ]}
]
variables [
 CurrentRotation {type=number}
 CurrentRotation_Duck {type=number}
 CurrentRotation_Duck_1 {type=number}
 Duck_1 Floating {type=Vector}
 Duck_1 rotator {type=Rotator}
 Duck_2 Floating {type=Vector}
 Duck_2 rotator {type=Rotator}
 Gravity Override {type=Vector}
]
